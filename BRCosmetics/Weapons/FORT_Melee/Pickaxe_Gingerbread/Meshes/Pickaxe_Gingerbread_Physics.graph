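# Physics asset for the Gingerbread pickaxe melee weapon, giving its dynamic bow bones physical bodies and joint limits.
# A capsule covers the two-handed handle, and box bodies cover the left and right BowA and BowB bones.
# Each BowA and first BowB segment is constrained to the handle, and each second BowB segment chains to its first.
# All joints lock twist and limit swing, with tighter swing limits on the BowA pieces.
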PhysicsAsset /BRCosmetics/Weapons/FORT_Melee/Pickaxe_Gingerbread/Meshes/Pickaxe_Gingerbread_Physics
bodies {
 Melee_TwoHanded_Handle [{shapeType=Capsule, length=99.8561, radius=1.7098, center=<0.8118 -0.9304 26.7465>, rotation=<rot -0.0548 -90.0057 0.3784>}]
 dyn_Left_BowA_01 [{shapeType=Box, extentX=9.9263, extentY=7.5, extentZ=2.0772, center=<1.2879 0.4289 -0.2376>, rotation=<rot -14.9999 0 0>}]
 dyn_Left_BowB_01 [{shapeType=Box, extentX=7.5, extentY=3.7048, extentZ=0.9014}]
 dyn_Left_BowB_02 [{shapeType=Box, extentX=7.5, extentY=3.7048, extentZ=0.9014, center=<0.7287 -0.1629 -0.0637>, rotation=<rot -5 -0.0003 0>}]
 dyn_Right_BowA_01 [{shapeType=Box, extentX=12.5745, extentY=9.6208, extentZ=3.2063, center=<2.3177 0.482 0.0002>}]
 dyn_Right_BowB_01 [{shapeType=Box, extentX=7.5, extentY=4.3204, extentZ=0.8285}]
 dyn_Right_BowB_02 [{shapeType=Box, extentX=7.5, extentY=4.9349, extentZ=0.8285, center=<1.0469 0 0.0001>}]
}
constraints [
 {
  bone1Name=dyn_Left_BowA_01
  bone2Name=Melee_TwoHanded_Handle
  swing1LimitDegrees=15
  swing2LimitDegrees=20
  twistLimitDegrees=45
  twistMotion=Locked
 }
 {
  bone1Name=dyn_Right_BowA_01
  bone2Name=Melee_TwoHanded_Handle
  swing1LimitDegrees=15
  swing2LimitDegrees=25
  twistLimitDegrees=45
  twistMotion=Locked
 }
 {
  bone1Name=dyn_Left_BowB_01
  bone2Name=Melee_TwoHanded_Handle
  swing1LimitDegrees=30
  swing2LimitDegrees=30
  twistLimitDegrees=45
  twistMotion=Locked
 }
 {
  bone1Name=dyn_Left_BowB_02
  bone2Name=dyn_Left_BowB_01
  swing1LimitDegrees=30
  swing2LimitDegrees=30
  twistLimitDegrees=45
  twistMotion=Locked
 }
 {
  bone1Name=dyn_Right_BowB_01
  bone2Name=Melee_TwoHanded_Handle
  swing1LimitDegrees=30
  swing2LimitDegrees=30
  twistLimitDegrees=45
  twistMotion=Locked
 }
 {
  bone1Name=dyn_Right_BowB_02
  bone2Name=dyn_Right_BowB_01
  swing1LimitDegrees=30
  swing2LimitDegrees=30
  twistLimitDegrees=45
  twistMotion=Locked
 }
]
